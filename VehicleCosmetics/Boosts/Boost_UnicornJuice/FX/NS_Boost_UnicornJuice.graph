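# Niagara system for the Unicorn Juice vehicle boost cosmetic, emitting a rainbow exhaust trail from the car's boost sockets.
# A sprite main trail and a second mesh-plus-sprite emitter partition particles across skeletal mesh sockets by port count.
# Default color hue-shifts a looping RGB curve, while painted-variant overrides supply per-paint colors.
# Trail velocity switches by game mode, with inherited source movement fading over particle life.
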NiagaraSystem /VehicleCosmetics/Boosts/Boost_UnicornJuice/FX/NS_Boost_UnicornJuice
dependencies {
 usedDataInterfaces [
  &NiagaraDataInterfaceArrayFloat
  &NiagaraDataInterfaceArrayColor
  &NiagaraDataInterfaceCurve
  &NiagaraDataInterfaceColorCurve
  &NiagaraDataInterfaceVectorCurve
  &NiagaraDataInterfaceSpriteRendererInfo
  &NiagaraDataInterfaceSkeletalMesh
 ]
 usedDynamicInputs []
 usedModules [
  &self:SystemSpawnScript.SystemScriptSource.SystemScriptGraph.NiagaraNodeAssignment_0.SetVariables_#1
  &SystemState
 ]
}
emitters {
 NE_CS_MainTrail {
  emitterSpawnScript [/SetVariables(
    Emitter.Linear Color:LinearColor=&SelectLinearColorFromArray(
     Array Sampling Mode=::NewEnumerator1
     Direct Array Index=$System.ColorVariantID
     Color Selection Array=DataInterface() {propertyValues {ColorData [
        <rgba 0 0 0 1>
        <rgba 1 1 1 1>
        <rgba 0.5312 0.5312 0.5312 1>
        <rgba 0.2812 0.2389 0.2212 1>
        <rgba 0.0677 0.2043 1 1>
        <rgba 0.5938 0 0.0152 1>
        <rgba 0.4665 0.7812 0.4985 1>
        <rgba 0.5521 0.5521 0.5521 1>
        <rgba 1 1 1 1>
        <rgba 1 1 1 1>
        <rgba 1 1 1 1>
        <rgba 1 1 1 1>
        <rgba 1 1 1 1>
        <rgba 1 1 1 1>
       ]}}
     Random Seed=?
    )
    Emitter.SpecularPainted:NiagaraFloat=&SelectFloatFromArray(
     Direct Array Index=$System.ColorVariantID
     Float Selection Array=DataInterface() {propertyValues {FloatData [0.7,0,0.7,0.7,1,0.5,0.7,0.5,1,1,1,1,1,1,1]}}
    )
   ) {isSetParameters=true}]
  emitterUpdateScript [
   &EmitterState()
   &VC_FilteredSocketCount()
   &SpawnRate(SpawnRate=90, Use Spawn Probability=false)
   &SpawnRate(SpawnGroup=1, SpawnRate=40, Use Spawn Probability=false) {module=SpawnRate001}
   &SpawnRate(SpawnGroup=2, SpawnRate=40, Use Spawn Probability=false) {module=SpawnRate002}
  ]
  particleSpawnScript [
   /SetVariables(Particles.VisTag:NiagaraInt32=&SpawnGroup()) {isSetParameters=true}
   &VC_PaintedVariantsOverride(
    R1 DefaultColor=<rgba 0.1315 0.4088 1 1>
    R2 Black:LinearColor=<rgba 0.16 0.16 0.16 1>
    R2 DefaultColor:LinearColor=<rgba 0.9348 0.5504 0.9896 1>
    Use R1 Painted Overrides=true
    Use R2 Painted Overrides=true
   ) {enabled=false}
   &InitializeParticle(
    Color=$Particles.Color
    Lifetime=&MaskFloat(
     Mask Source=0
     Mask Parameter=?
     Mask=0
     Masked Float=&RandomRangeFloat(Minimum=0.75, Maximum=1)
     Passthrough Float=&RandomRangeFloat(Minimum=0.3, Maximum=0.6)
    )
    Sprite Rotation Angle=&RandomRangeFloat(Minimum=-1, Maximum=1)
    Sprite Rotation Mode=::NewEnumerator1
    Sprite Size Mode=::NewEnumerator3
    Uniform Sprite Size=&MaskFloat(
     Mask Source=0
     Mask Parameter=?
     Mask=0
     Masked Float=&RandomRangeFloat(Minimum=140, Maximum=180)
     Passthrough Float=&RandomRangeFloat(Minimum=80, Maximum=120)
    )
    UsePositionOffset=false
    Write Material Random=true
   )
   &PartitionParticles(Number Of Partitions=$Emitter.PortCount)
   &SkeletalMeshLocation(
    Bone / Socket Index=$Particles.PartitionParticles.Partition
    Bone / Socket Sampling Mode=::NewEnumerator5
    MaskSpawnGroup=false
    Mesh Sampling Type=::NewEnumerator6
    Sampled Position Offset=&AddVectorToPosition(Position=<0 0 0>, Vector=$User.DTCarOffset)
    Skeletal Mesh=$System.SkeletalMesh
   )
   &AddVelocity(Velocity=&MaskVector(
     Mask Source=0
     Mask Parameter=?
     Mask=0
     Masked Vector=<-1400 0 50>
     Passthrough Vector=&MergeHelper_VectorSwitch(
      0=&RandomRangeVector(Minimum=<-1000 0 -96>, Maximum=<-1000 0 122>)
      1=&RandomRangeVector(Minimum=<-1000 0 -96>, Maximum=<-1000 0 122>)
      2=&RandomRangeVector(Minimum=<0 0 -96>, Maximum=<0 0 122>)
      3=?
      Index=$User.GameModeID
      Num=::NewEnumerator1
     )
    ))
  ]
  particleUpdateScript [
   &ParticleState()
   &VC_PaintedVariantsOverride(
    Black=&SelectLinearColorFromArray(
     Array Sampling Mode=::NewEnumerator1
     Direct Array Index=&RandomRangeInt(Minimum=0, Maximum=1)
     Color Selection Array=DataInterface() {propertyValues {ColorData [<rgba 0.015 0.015 0.015 1>, <rgba 0.0417 0.0417 0.0417 1>]}}
     Random Seed=?
    )
    BurntSienna=&RandomRangeLinearColor(
     Minimum=&ColorFromCurve(ColorCurve=DataInterface() {propertyValues {
        AlphaCurve {keys [<at 0 1 RCIM_Cubic>]}
        BlueCurve {keys [<at 0.0012 0.2204>,<at 0.1917 0.026 RCIM_Cubic>]}
        GreenCurve {keys [<at 0.0012 0.2935>,<at 0.1917 0.0661 RCIM_Cubic>]}
        RedCurve {keys [<at 0.0012 0.651>,<at 0.1917 0.2623 RCIM_Cubic>]}
       }})
     Maximum=<rgba 0.099 0.0251 0.0103 1>
     Color Channel Mode=::NewEnumerator0
    )
    Cobalt=&RandomRangeLinearColor(Minimum=<rgba 0.1198 0.1948 1 1>, Maximum=<rgba 0.0195 0.038 0.375 1>)
    Crimson=&RandomRangeLinearColor(Minimum=<rgba 0.1562 0.0031 0.0073 1>, Maximum=<rgba 0.5043 0.0482 0.0606 1>)
    DefaultColor=&HueShiftLinearColor(
     Hue=&FloatFromCurve(FloatCurve=DataInterface() {propertyValues {Curve {keys [<at 0 0 RCIM_Cubic>,<at 0.6 1 RCIM_Cubic>]}}})
     Saturation=1
     Value=0.9
     Linear Color=&MultiplyLinearColor(
      Linear Color=&ColorFromCurve(
       ColorCurve=DataInterface() {propertyValues {
         AlphaCurve {keys [<at 0 1>,<at 1 1>]}
         BlueCurve {keys [<at 0 0>,<at 0.333 0>,<at 0.667 1>,<at 1 0>]}
         GreenCurve {keys [<at 0 0>,<at 0.333 1>,<at 0.667 0>,<at 1 0>]}
         RedCurve {keys [<at 0 1>,<at 0.333 0>,<at 0.667 0>,<at 1 1>]}
        }}
       CurveIndex=$Emitter.NormalizedLoopAge
      )
      Scale Factor=&MakeVector4(
       X=1
       Y=1
       Z=1
       W=&FloatFromCurve(FloatCurve=DataInterface() {propertyValues {Curve {keys [
            <at 0 0 RCIM_Cubic>
            <at 0.1 1 RCIM_Cubic>
            {arriveTangent=-4.8117, interpMode=RCIM_Cubic, leaveTangent=-4.8117, tangentMode=RCTM_User, time=0.9, value=0}
           ]}}})
      )
     )
    )
    ForestGreen=&RandomRangeLinearColor(Minimum=<rgba 0.0859 0.4583 0.1048 1>, Maximum=<rgba 0.0029 0.0938 0.0075 1>)
    Grey=&RandomRangeLinearColor(Minimum=<rgba 0.0938 0.0938 0.0938 1>, Maximum=<rgba 0.4427 0.4427 0.4427 1>)
    Lime=&RandomRangeLinearColor(Minimum=<rgba 0.1875 0.375 0 1>, Maximum=<rgba 0.2943 0.5885 0 1>)
    Orange=&RandomRangeLinearColor(Minimum=<rgba 0.3177 0.1108 0 1>, Maximum=<rgba 0.7292 0.2543 0 1>)
    Pink=&RandomRangeLinearColor(Minimum=<rgba 0.375 0.0664 0.2628 1>, Maximum=<rgba 0.7396 0.1772 0.5351 1>)
    Purple=&RandomRangeLinearColor(Minimum=<rgba 0.1902 0 0.4635 1>, Maximum=<rgba 0.2715 0 0.6615 1>)
    Saffron=&RandomRangeLinearColor(Minimum=<rgba 0.474 0.3634 0 1>, Maximum=<rgba 0.75 0.5868 0.0508 1>)
    SkyBlue=<rgba 0.0592 0.3109 0.4375 1>
    TitaniumWhite=<rgba 0.5156 0.5156 0.5156 1>
   ) {module=VC_PaintedVariantsOverride001}
   &AccelerationForce(Acceleration=&RandomRangeVector(Minimum=<0 0 -128>, Maximum=<1 1 -187.5>))
   &ScaleSpriteSize(
    Uniform Curve Scale=2
    Uniform Curve Sprite Scale=DataInterface() {propertyValues {Curve {keys [
        <at 0 5 RCIM_Cubic>
        {arriveTangent=0.6117, interpMode=RCIM_Cubic, leaveTangent=0.6117, tangentMode=RCTM_User, time=0.3, value=8}
        <at 1 15 RCIM_Cubic>
       ]}}}
   ) {enabled=false}
   &Drag(
    Drag=&FloatFromCurve(
     FloatCurve=DataInterface() {propertyValues {Curve {
        defaultValue=1
        keys [
         {arriveTangent=0.0291, interpMode=RCIM_Cubic, leaveTangent=0.003, tangentMode=RCTM_Break, time=0, value=0}
         {arriveTangent=4.1537, interpMode=RCIM_Cubic, leaveTangent=4.1537, tangentMode=RCTM_User, time=0.5, value=2}
        ]
       }}}
     Scale Curve=&MergeHelper_FloatSwitch(0=20, 1=20, 2=20, 3=?, Index=$User.GameModeID, Num=::NewEnumerator1)
    )
    Use Linear Drag=true
    Use Rotational Drag=true
   ) {enabled=false}
   &ScaleColor(
    Scale Alpha=&FloatFromCurve(FloatCurve=DataInterface() {propertyValues {Curve {keys [
         <at 0 3 RCIM_Cubic>
         {arriveTangent=0.0053, interpMode=RCIM_Cubic, leaveTangent=0.0053, tangentMode=RCTM_User, time=0.8, value=1}
         <at 1 0 RCIM_Cubic>
        ]}}})
    Scale RGB=&VectorFromFloat(Value=&MakeCustomFloatFromBool(Bool=$User.bIsPainted, True Float=&RandomRangeFloat(Minimum=0.5, Maximum=2), False Float=1.2))
    ScaleA=true
    ScaleRGB=true
   ) {enabled=false}
   &InheritSourceMovement(
    Applied Position Scale=&VectorFromCurve(VectorCurve=DataInterface() {propertyValues {
       XCurve {keys [<at 0 1 RCIM_Cubic>,<at 0.5 0 RCIM_Cubic>]}
       YCurve {keys [<at 0 1 RCIM_Cubic>,<at 0.5 0 RCIM_Cubic>]}
       ZCurve {keys [<at 0 1 RCIM_Cubic>,<at 0.5 0 RCIM_Cubic>]}
      }})
    ApplyForce=false
    ApplyPosition=true
    ApplyPositionFalloff=false
   ) {module=InheritSourceMovement001}
   &ScaleVelocity(
    Coordinate Space:ENiagaraCoordinateSpace=::Simulation
    Velocity Scale=&VectorFromFloat(Value=&FloatFromCurve(FloatCurve=DataInterface() {propertyValues {Curve {keys [<at 0 0.8 RCIM_Cubic>,<at 1 0>]}}}))
   )
   &InheritSourceMovement(
    Applied Position Scale=&VectorFromFloat(Value=&FloatFromCurve(FloatCurve=DataInterface() {propertyValues {Curve {keys [
          {interpMode=RCIM_Cubic, tangentMode=RCTM_Break, time=0, value=1}
          {arriveTangent=-2.9025, interpMode=RCIM_Cubic, tangentMode=RCTM_Break, time=1, value=0}
         ]}}}))
    ApplyForce=false
    ApplyPosition=true
    ApplyPositionFalloff=false
   ) {enabled=false}
   &SolveForcesAndVelocity(Clamp Velocity=false, Limit Acceleration=false)
   &SubUVAnimation(
    Output Alternate Renderer Binding=true
    "Sprite Renderer "=DataInterface() {propertyValues {SpriteRenderer=&self:NE_CS_MainTrail_0.NiagaraSpriteRendererProperties_1}}
    UseEndFrame=false
    UseStartFrame=false
   )
   &DynamicMaterialParameters(
    Index 0 Param 1=1
    Module.Index0Float0Write=true
    Module.Index0Float1Write=true
    Module.Index0Float2Write=true
    Module.Index0Float3Write=true
   )
  ]
  rendererClasses [&NiagaraSpriteRendererProperties]
  renderers [{rendererClass=&NiagaraSpriteRendererProperties}]
 }
 NE_VC_Boost_Sprite_Parent {
  emitterUpdateScript [
   &EmitterState()
   &VC_FilteredSocketCount()
   &SpawnRate(SpawnRate=40, Use Spawn Probability=false)
  ]
  particleSpawnScript [
   &VC_PaintedVariantsOverride() {enabled=false}
   &InitializeParticle(
    Color=$Particles.Color
    Lifetime Max=0.4
    Lifetime Min=0.3
    Lifetime Mode=::NewEnumerator1
    Mesh Scale Max=<2.5 3 4>
    Mesh Scale Min=<1.125 1.125 2.25>
    Mesh Scale Mode=::NewEnumerator2
    Sprite Rotation Mode=::NewEnumerator2
    UsePositionOffset=false
    Write Material Random=true
   )
   &ShapeLocation(MaskSpawnGroup=false, Override Seed=false, Sphere Radius=1, UseOwnerScale=true)
   &PartitionParticles(Number Of Partitions=$Emitter.PortCount)
   &SkeletalMeshLocation(
    Bone / Socket Index=$Particles.PartitionParticles.Partition
    Bone / Socket Sampling Mode=::NewEnumerator5
    MaskSpawnGroup=false
    Mesh Sampling Type=::NewEnumerator6
    Skeletal Mesh=$System.SkeletalMesh
   )
   &AddVelocity(Velocity=&RandomRangeVector(Minimum=<-100 -15 -10>, Maximum=<-250 15 15>))
   &InitialMeshOrientation(
    Mesh Axis To Orient=<-1 0 0>
    Mesh Orientation Mode=::NewEnumerator2
    Orientation Coordinate Space=::Local
    Orientation Vector=<0 0 1>
    Rotation=&RandomRangeVector(Minimum=<-1 0 0>, Maximum=<1 0 0>)
    Rotation Coordinate Space=::NewEnumerator0
    Use Rotation Vector=true
   )
  ]
  particleUpdateScript [
   &ParticleState()
   &VC_PaintedVariantsOverride(
    Black=<rgba 0.018 0.018 0.018 1>
    BurntSienna=<rgba 0.2604 0.0583 0.0041 1>
    Cobalt=<rgba 0.0781 0.0952 1 1>
    Crimson=<rgba 0.2865 0.0143 0.0143 1>
    DefaultColor=&HueShiftLinearColor(
     Hue=&FloatFromCurve(FloatCurve=DataInterface() {propertyValues {Curve {keys [<at 0 0 RCIM_Cubic>,<at 0.6 1 RCIM_Cubic>]}}})
     Saturation=1
     Value=1
     Linear Color=&MultiplyLinearColor(
      Linear Color=&ColorFromCurve(
       ColorCurve=DataInterface() {propertyValues {
         AlphaCurve {keys [<at 0 1>,<at 1 1>]}
         BlueCurve {keys [<at 0 0>,<at 0.333 0>,<at 0.667 1>,<at 1 0>]}
         GreenCurve {keys [<at 0 0>,<at 0.333 1>,<at 0.667 0>,<at 1 0>]}
         RedCurve {keys [<at 0 1>,<at 0.333 0>,<at 0.667 0>,<at 1 1>]}
        }}
       CurveIndex=$Emitter.NormalizedLoopAge
      )
      Scale Factor=&MakeVector4(
       X=1
       Y=1
       Z=1
       W=&FloatFromCurve(FloatCurve=DataInterface() {propertyValues {Curve {keys [
            <at 0 0 RCIM_Cubic>
            <at 0.1 1 RCIM_Cubic>
            {arriveTangent=-4.8117, interpMode=RCIM_Cubic, leaveTangent=-4.8117, tangentMode=RCTM_User, time=0.9, value=0}
           ]}}})
      )
     )
    )
    ForestGreen=<rgba 0.0001 0.25 0.0004 1>
    Grey=<rgba 0.7043 0.5043 0.4375 1>
    Lime=<rgba 0.3039 0.3646 0.0057 1>
    Orange=<rgba 0.5104 0.2113 0.051 1>
    Pink=<rgba 0.5052 0.1552 0.3779 1>
    Purple=<rgba 0.1753 0 0.4271 1>
    Saffron=<rgba 0.3385 0.269 0.0406 1>
    SkyBlue=<rgba 0.0894 0.4527 0.6354 1>
    TitaniumWhite=<rgba 0.3646 0.3646 0.3646 1>
   ) {module=VC_PaintedVariantsOverride001}
   &ScaleMeshSize(Scale Factor=&VectorFromCurve(VectorCurve=DataInterface() {propertyValues {
       XCurve {keys [
         <at 0 0 RCIM_Cubic>
         {arriveTangent=-0.0318, interpMode=RCIM_Cubic, leaveTangent=-0.0318, tangentMode=RCTM_User, time=0.1, value=1}
         <at 1 2 RCIM_Cubic>
        ]}
       YCurve {keys [
         <at 0 0 RCIM_Cubic>
         {arriveTangent=-0.0318, interpMode=RCIM_Cubic, leaveTangent=-0.0318, tangentMode=RCTM_User, time=0.1, value=1}
         <at 1 2 RCIM_Cubic>
        ]}
       ZCurve {keys [
         <at 0 0 RCIM_Cubic>
         {arriveTangent=-0.0318, interpMode=RCIM_Cubic, leaveTangent=-0.0318, tangentMode=RCTM_User, time=0.1, value=1}
         <at 1 2 RCIM_Cubic>
        ]}
      }}))
   &DynamicMaterialParameters(
    Index 0 Param 1=1
    Index 0 Param 4=&FloatFromCurve(FloatCurve=DataInterface() {propertyValues {Curve {keys [<at 0.5 0>,<at 1 1>]}}})
    Module.Index0Float0Write=true
    Module.Index0Float1Write=true
    Module.Index0Float2Write=true
    Module.Index0Float3Write=true
   )
   &SolveForcesAndVelocity(Clamp Velocity=false, Limit Acceleration=false)
  ]
  rendererClasses [&NiagaraMeshRendererProperties, &NiagaraSpriteRendererProperties]
  renderers [
   {rendererClass=&NiagaraMeshRendererProperties, rendererIndex=0}
   {rendererClass=&NiagaraSpriteRendererProperties, rendererIndex=1}
  ]
 }
}
summary {
 systemName=NS_Boost_UnicornJuice
 emitters [
  {emitterName=NE_CS_MainTrail, rendererClasses [&NiagaraSpriteRendererProperties]}
  {
   emitterName=NE_VC_Boost_Sprite_Parent
   rendererClasses [&NiagaraMeshRendererProperties, &NiagaraSpriteRendererProperties]
  }
 ]
 userVariables [
  User.bIsFrontendPreview {type=NiagaraBool}
  User.bIsPainted {type=NiagaraBool}
  User.ColorVariantID {type=NiagaraInt32}
  User.DTCarOffset {type=Vector3f, description="", defaultValue {struct=Vector3f, value=<0 0 0>}}
  User.GameModeID {type=NiagaraInt32}
  User.PaintedColor {type=LinearColor, description="", defaultValue {struct=LinearColor, value=<rgba 1 1 1 1>}}
  User.PaintedScalar {type=NiagaraFloat, defaultValue=1, description=""}
  User.PortCount {type=NiagaraInt32, defaultValue=1}
 ]
}
